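; Blueprint function library of shared helpers for Fortnite gameplay abilities.
; Math helpers cover root-motion duration, angles, actor distances, range-to-tile conversion and flattened teleport projection along velocity, falling back to facing.
; Ability helpers read effect-container ranges, resolve perk-tier values, count effects, find player-state pawns and gather target-data hit actors, with one array builder marked do-not-use.
; ApplySlowFall zeroes avatar velocity, then applies an owner effect with SetByCaller duration.
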(fn ApplySlowFall
  (params (SlowFallDuration :type Float%20%28double-precision%29) (SelfGameplayAbilityReference :type Fort%20Gameplay%20Ability%20Object%20Reference) (GameplayEffect :type Gameplay%20Effect%20Class%20Reference))
  (bind _asfort_pawn (Utilities|Casting|CastToFortPawn (Ability|GetAvatarActorfromActorInfo SelfGameplayAbilityReference)))
  (Variables|/Script/Engine.MovementComponent|SetVelocity_X (0.0 0.0 0.0) (Variables|/Script/Engine.Character|GetCharacterMovement _asfort_pawn))
  (bind _returnvalue (Ability|GameplayEffect|AssignTagSetbyCallerMagnitude (Ability|MakeOutgoingGameplayEffectSpec SelfGameplayAbilityReference GameplayEffect 1.0) "(TagName=\"SetByCaller.Duration\")" SlowFallDuration))
  (bind _returnvalue_2 (Ability|ApplyGameplayEffectSpecToOwner SelfGameplayAbilityReference _returnvalue))
  (return _returnvalue_2))

(fn CalculateRootMotionDuration
  (params (Distance :type Float%20%28double-precision%29) (Speed :type Float%20%28double-precision%29) (ClampMinMax? :type Boolean) (MinDuration :type Float%20%28double-precision%29) (MaxDuration :type Float%20%28double-precision%29))
  ;[0.4895 0.1317 1] Provide a calculation for Root Motion duration variables using distance and target speed
  (return (Math|Float|SelectFloat (Math|Float|Clamp(Float) (Math|Float|SafeDivide Distance Speed) MinDuration MaxDuration) (Math|Float|SafeDivide Distance Speed) ClampMinMax?)))

(fn ConvertRangeToTiles
  (params (InRange :type Float%20%28double-precision%29))
  (return (/ InRange "512.0f")))

(fn DoNOTUse-BuildArrayOfHitActors
  (params (TargetData :type Gameplay%20Ability%20Target%20Data%20Handle%20Structure))
  (for _index (range (Ability|TargetData|GetDataCountfromTargetData TargetData))
    (if (Utilities|NotEqual(Object) (Utilities|Array|Get(acopy) (Ability|TargetData|GetActorsfromTargetData TargetData _index)) 0)
      (Utilities|Array|Add (LocalVariables|GetLocalHitActors) (Utilities|Array|Get(acopy) (Ability|TargetData|GetActorsfromTargetData TargetData _index)))))
  (return (LocalVariables|GetLocalHitActors)))

(fn EvaluatePerkTierBalanceValue
  (params (T01%20Balance%20Value :type Scalable%20Float%20Structure) (T02%20Balance%20Value :type Scalable%20Float%20Structure) (Owns%20T02 :type Boolean) (Should%20Override%20T01 :type Boolean))
  (return (select (Variables|GetShould%20Override%20T01) (select (Variables|GetOwns%20T02) (Variables|GetT02%20Balance%20Value) (Variables|GetT01%20Balance%20Value)) (Utilities|Struct|MakeScalableFloat (select (Variables|GetOwns%20T02) (+ (ScalableFloat|GetValueatLevel (Variables|GetT01%20Balance%20Value)) (ScalableFloat|GetValueatLevel (Variables|GetT02%20Balance%20Value))) (ScalableFloat|GetValueatLevel (Variables|GetT01%20Balance%20Value)))))))

(fn GetAngleBetweenTwoVectors
  (params (VectorA :type Vector) (VectorB :type Vector))
  (return (Math|Trig|Acos(Degrees) (Math|Vector|DotProduct (Math|Vector|Normalize VectorA) (Math|Vector|Normalize VectorB)))))

(fn GetDistanceBetweenActors
  (params (Actor1 :type Actor%20Object%20Reference) (Actor2 :type Actor%20Object%20Reference))
  (if (and (Utilities|IsValid Actor1 :!class /Script/Engine.KismetSystemLibrary) (Utilities|IsValid Actor2 :!class /Script/Engine.KismetSystemLibrary))
    (return (Math|Vector|VectorLength (- (Transformation|GetActorLocation Actor1) (Transformation|GetActorLocation Actor2))))
    (else
      (return 0.0))))

(fn GetDistanceFromOwnerToAITarget
  (params (Ability :type Fort%20Gameplay%20Ability%20Object%20Reference))
  (return (CallFunction|GetDistanceBetweenActors (Ability|GetActivatingPawn Ability) (Ability|GetAIAbilityTarget Ability) (Ability|GetActivatingPawn Ability))))

(fn GetDistanceSquaredBetweenActors
  (params (Actor%201 :type Actor%20Object%20Reference) (Actor%202 :type Actor%20Object%20Reference))
  (return (Math|Vector|VectorLengthSquared (- (Transformation|GetActorLocation Actor%201) (Transformation|GetActorLocation Actor%202)))))

(fn GetPawnFromPlayerStateZone
  (params (InActor :type Actor%20Object%20Reference))
  ;[1 1 1] Gets the pawn from the Player Sate Zone
  (Utilities|IsValid InActor
    (:Is%20Valid
      (bind _asfort_player_state_zone (Utilities|Casting|CastToFortPlayerStateZone InActor)
        (:then
          (Utilities|IsValid (PlayerState|GetCurrentPawn _asfort_player_state_zone)
            (:Is%20Valid
              (return (PlayerState|GetCurrentPawn _asfort_player_state_zone) true))
            (:Is%20Not%20Valid
              _)))
        (:CastFailed
          _))
      (return 0 false))
    (:Is%20Not%20Valid
      _)))

(fn GetProjectedTeleportLocation
  (params (Actor :type Actor%20Object%20Reference) (MovementDistance :type Float%20%28double-precision%29))
  (return (+ (Transformation|GetActorLocation Actor) (* (Math|Vector|Normalize (Math|Vector|MakeVector (.x (Math|Vector|SelectVector (Transformation|GetActorForwardVector Actor) (Transformation|GetVelocity Actor) (or (== (Math|Vector|VectorLength (Transformation|GetVelocity Actor)) 0.0) (and (== (.x (Transformation|GetVelocity Actor)) 0.0) (== (.y (Transformation|GetVelocity Actor)) 0.0))))) (.y (Math|Vector|SelectVector (Transformation|GetActorForwardVector Actor) (Transformation|GetVelocity Actor) (or (== (Math|Vector|VectorLength (Transformation|GetVelocity Actor)) 0.0) (and (== (.x (Transformation|GetVelocity Actor)) 0.0) (== (.y (Transformation|GetVelocity Actor)) 0.0))))))) MovementDistance))))

(fn GetRangeFromGameplayEffectContainer
  (params (GE%20Container :type Fort%20Gameplay%20Effect%20Container%20Structure) (Target%20Selection%20List%20Index :type Integer) (At%20Level :type Float%20%28double-precision%29))
  (bind (_shape _customshapecomponentname _testtype _primarysource _secondarysource _range) (Utilities|Struct|BreakFortAbilityTargetSelection (Utilities|Array|Get(acopy) (Utilities|Struct|BreakFortAbilityTargetSelectionList (.TargetSelection (Utilities|Struct|BreakFortGameplayEffectContainer GE%20Container))) Target%20Selection%20List%20Index)))
  (bind (_applicationtag _targetselection) (Utilities|Struct|BreakFortGameplayEffectContainer GE%20Container))
  (return (ScalableFloat|GetValueatLevel _range At%20Level)))

(fn GetRangeFromGameplayEffectContainerSpec
  (params (GE%20Container%20Spec :type Fort%20Gameplay%20Effect%20Container%20Spec%20Structure) (Target%20Selection%20List%20Index :type Integer) (At%20Level :type Float%20%28double-precision%29))
  (bind (_shape _customshapecomponentname _testtype _primarysource _secondarysource _range) (Utilities|Struct|BreakFortAbilityTargetSelection (Utilities|Array|Get(acopy) (Utilities|Struct|BreakFortAbilityTargetSelectionList (Utilities|Struct|BreakFortGameplayEffectContainerSpec GE%20Container%20Spec)) Target%20Selection%20List%20Index)))
  (return (ScalableFloat|GetValueatLevel _range At%20Level)))

(fn GetTotalGameplayEffectCount
  (params (EffectsToCount :type Array%20of%20Gameplay%20Effect%20Class%20References) (AbilitySystemComponent :type Ability%20System%20Component%20Object%20Reference))
  (for _array_element EffectsToCount
    (LocalVariables|SetTempTotalEffectCount (+ (LocalVariables|GetTempTotalEffectCount) (GameplayEffects|GetGameplayEffectCount AbilitySystemComponent _array_element 0 true))))
  (return (LocalVariables|GetTempTotalEffectCount)))

(fn SetArrayOfHitActors
  (params (TargetData :type Gameplay%20Ability%20Target%20Data%20Handle%20Structure) (ArrayToSet :type Array%20of%20Actor%20Object%20References))
  (Utilities|Array|Clear ArrayToSet)
  (for _index (range (Ability|TargetData|GetDataCountfromTargetData TargetData))
    (if (Utilities|NotEqual(Object) (Utilities|Array|Get(acopy) (Ability|TargetData|GetActorsfromTargetData TargetData _index)) 0)
      (Utilities|Array|Add ArrayToSet (Utilities|Array|Get(acopy) (Ability|TargetData|GetActorsfromTargetData TargetData _index))))))
