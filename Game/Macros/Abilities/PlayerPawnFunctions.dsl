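; Player pawn ability helper that checks whether a character material derives from the transparent character material used for invisibility.
; For material instances, a while loop climbs the parent chain to the base Material; plain materials are compared directly.
; It returns the negated match, so true means not transparent, plus a validity flag cleared when no base material resolves.
; An invalid input material skips the check entirely.
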(fn CheckToSeeIfCharacterMaterialIsInvisible
  (params (Current%20Material :type Material%20Interface%20Object%20Reference))
  (bind _currentmaterialinstance (LocalVariables|GetCurrentMaterialInstance))
  (bind _parent (Variables|/Script/Engine.MaterialInstance|GetParent _currentmaterialinstance))
  (Utilities|IsValid Current%20Material
    (:Is%20Valid
      (bind _asmaterial_instance (Utilities|Casting|CastToMaterialInstance Current%20Material)
        (:then
          (LocalVariables|SetCurrentMaterialInstance _asmaterial_instance)
          (LocalVariables|SetExecute%20While%20Loop true)
          (while (LocalVariables|GetExecute%20While%20Loop)
            ;[0 0 0] Find base material and see if it is the transparent character mat
            (Utilities|IsValid _parent
              (:Is%20Valid
                (bind _asmaterial (Utilities|Casting|CastToMaterial _parent)
                  (:then
                    (LocalVariables|SetMaterialParentIsTranslucent (Utilities|Equal(Object) _asmaterial (LocalVariables|GetTransparentCharacterMaterial)))
                    (LocalVariables|SetExecute%20While%20Loop false))
                  (:CastFailed
                    (LocalVariables|SetCurrentMaterialInstance (Utilities|Casting|CastToMaterialInstance _parent)))))
              (:Is%20Not%20Valid
                (LocalVariables|SetExecute%20While%20Loop false)
                (LocalVariables|SetvalidMaterial false)))))
        (:CastFailed
          ;[0 0 0] cast to base material class
          (bind _asmaterial_2 (Utilities|Casting|CastToMaterial Current%20Material)
            (:then
              (LocalVariables|SetMaterialParentIsTranslucent (Utilities|Equal(Object) _asmaterial_2 (LocalVariables|GetTransparentCharacterMaterial))))
            (:CastFailed
              (LocalVariables|SetvalidMaterial false)))))
      (return (not (LocalVariables|GetMaterialParentIsTranslucent)) (LocalVariables|GetvalidMaterial)))
    (:Is%20Not%20Valid
      _)))
